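# Animation Blueprint for the BrawnyBass Style male large player body cosmetic, layering secondary motion onto the pose copied from the attached parent mesh.
# Weighted multi-bone constraints drive shoulder and pouch master bones, then spring-based Anim Dynamics swings the pouch in root-relative space.
# Two Control Rigs follow: the Style rig, then the base BrawnyBass arms rig.
# Each update sets a 0/1 skydive-or-emote value that feeds the Style rig.
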AnimBlueprint /BRCosmetics/Characters/Player/Male/Large/Bodies/M_LRG_BrawnyBass_Style/Meshes/M_LRG_BrawnyBass_Style_AnimBP
target_skeleton=&/M_LRG_BrawnyBass_Style_Skeleton
graphs [
 AnimGraph {nodes [
   n1 = "Constraint - dyn_shoulder_l_MSTR (PPPPPPPPP)"!(
    ComponentPose=Local To Component(LocalPose=Copy Pose From Mesh() {bUseAttachedParent=true})
    Alpha=1
    ConstraintWeights_0=0.024
    ConstraintWeights_1:float=0.0051
    ConstraintWeights_2:float=0.1745
    ConstraintWeights_3:float=0.4291
    ConstraintWeights_4:float=0.0424
    ConstraintWeights_5:float=0.1814
    ConstraintWeights_6:float=0.0149
    ConstraintWeights_7:float=0.0078
    ConstraintWeights_8:float=0.1207
   ) -> (Pose) {
    alphaBoolBlend {blendOption=Linear}
    boneToModify {boneName=dyn_shoulder_l_MSTR}
    constraintSetup [
     {targetBone {boneName=spine_03}}
     {targetBone {boneName=spine_04}}
     {targetBone {boneName=spine_05}}
     {targetBone {boneName=clavicle_l}}
     {targetBone {boneName=upperarm_twist_01_l}}
     {targetBone {boneName=deform_upperarm_up_l}}
     {targetBone {boneName=deform_upperarm_bk_l}}
     {targetBone {boneName=dfrm_l_arm_02}}
     {targetBone {boneName=dfrm_l_arm_01}}
    ]
    constraintWeights [1,1,1,1,1,1,1,1,1]
   }
   n2 = "Constraint - dyn_shoulder_r_MSTR (PPPPPPPPP)"!(
    ComponentPose=@n1
    Alpha=1
    ConstraintWeights_0=0.0231
    ConstraintWeights_1:float=0.0043
    ConstraintWeights_2:float=0.183
    ConstraintWeights_3:float=0.4637
    ConstraintWeights_4:float=0.0344
    ConstraintWeights_5:float=0.1424
    ConstraintWeights_6:float=0.0162
    ConstraintWeights_7:float=0.1258
    ConstraintWeights_8:float=0.0071
   ) -> (Pose) {
    alphaBoolBlend {blendOption=Linear}
    boneToModify {boneName=dyn_shoulder_r_MSTR}
    constraintSetup [
     {targetBone {boneName=spine_03}}
     {targetBone {boneName=spine_04}}
     {targetBone {boneName=spine_05}}
     {targetBone {boneName=clavicle_r}}
     {targetBone {boneName=upperarm_twist_01_r}}
     {targetBone {boneName=deform_upperarm_up_r}}
     {targetBone {boneName=deform_upperarm_bk_r}}
     {targetBone {boneName=dfrm_r_arm_01}}
     {targetBone {boneName=dfrm_r_arm_02}}
    ]
    constraintWeights [1,1,1,1,1,1,1,1,1]
   }
   n3 = "Constraint - dyn_shoulder_bk_l_MSTR (PPPPPPPP)"!(
    ComponentPose=@n2
    Alpha=1
    ConstraintWeights_0=0.1508
    ConstraintWeights_1:float=0.1478
    ConstraintWeights_2:float=0.2167
    ConstraintWeights_3:float=0.2872
    ConstraintWeights_4:float=0.0678
    ConstraintWeights_5:float=0.0256
    ConstraintWeights_6:float=0.0546
    ConstraintWeights_7:float=0.0493
   ) -> (Pose) {
    alphaBoolBlend {blendOption=Linear}
    boneToModify {boneName=dyn_shoulder_bk_l_MSTR}
    constraintSetup [
     {targetBone {boneName=spine_03}}
     {targetBone {boneName=spine_04}}
     {targetBone {boneName=spine_05}}
     {targetBone {boneName=clavicle_l}}
     {targetBone {boneName=upperarm_twist_01_l}}
     {targetBone {boneName=deform_upperarm_up_l}}
     {targetBone {boneName=deform_upperarm_bk_l}}
     {targetBone {boneName=dfrm_l_arm_01}}
    ]
    constraintWeights [1,1,1,1,1,1,1,1]
   }
   n4 = "Constraint - dyn_shoulder_bk_r_MSTR (PPPPPPP)"!(
    ComponentPose=@n3
    Alpha=1
    ConstraintWeights_0=0.1561
    ConstraintWeights_1:float=0.1487
    ConstraintWeights_2:float=0.211
    ConstraintWeights_3:float=0.3049
    ConstraintWeights_4:float=0.0729
    ConstraintWeights_5:float=0.034
    ConstraintWeights_6:float=0.0725
   ) -> (Pose) {
    alphaBoolBlend {blendOption=Linear}
    boneToModify {boneName=dyn_shoulder_bk_r_MSTR}
    constraintSetup [
     {targetBone {boneName=spine_03}}
     {targetBone {boneName=spine_04}}
     {targetBone {boneName=spine_05}}
     {targetBone {boneName=clavicle_r}}
     {targetBone {boneName=upperarm_twist_01_r}}
     {targetBone {boneName=deform_upperarm_up_r}}
     {targetBone {boneName=deform_upperarm_bk_r}}
    ]
    constraintWeights [1,1,1,1,1,1,1]
   }
   n5 = "Constraint - dyn_shoulder_fr_r_MSTR (PPPPPPPPP)"!(
    ComponentPose=@n4
    Alpha=1
    ConstraintWeights_0=0.0073
    ConstraintWeights_1:float=0.0287
    ConstraintWeights_2:float=0.0908
    ConstraintWeights_3:float=0.1269
    ConstraintWeights_4:float=0.2217
    ConstraintWeights_5:float=0.1076
    ConstraintWeights_6:float=0.138
    ConstraintWeights_7:float=0.0237
    ConstraintWeights_8:float=0.2553
   ) -> (Pose) {
    alphaBoolBlend {blendOption=Linear}
    boneToModify {boneName=dyn_shoulder_fr_r_MSTR}
    constraintSetup [
     {targetBone {boneName=spine_02}}
     {targetBone {boneName=spine_03}}
     {targetBone {boneName=spine_04}}
     {targetBone {boneName=spine_05}}
     {targetBone {boneName=clavicle_r}}
     {targetBone {boneName=upperarm_twist_01_r}}
     {targetBone {boneName=deform_upperarm_fr_r}}
     {targetBone {boneName=deform_upperarm_up_r}}
     {targetBone {boneName=deform_pec_r}}
    ]
    constraintWeights [1,1,1,1,1,1,1,1,1]
   }
   n6 = "Constraint - dyn_pouch_MSTR (PPPPPPPP)"!(
    ComponentPose=@n5
    Alpha=1
    ConstraintWeights_0=0.6138
    ConstraintWeights_1:float=0.114
    ConstraintWeights_2:float=0.1417
    ConstraintWeights_3:float=0.0854
    ConstraintWeights_4:float=0.0027
    ConstraintWeights_5:float=0.0003
    ConstraintWeights_6:float=0.0237
    ConstraintWeights_7:float=0.0186
   ) -> (Pose) {
    alphaBoolBlend {blendOption=Linear}
    boneToModify {boneName=dyn_pouch_MSTR}
    constraintSetup [
     {targetBone {boneName=pelvis}}
     {targetBone {boneName=spine_01}}
     {targetBone {boneName=spine_02}}
     {targetBone {boneName=spine_03}}
     {targetBone {boneName=spine_04}}
     {targetBone {boneName=clavicle_r}}
     {targetBone {boneName=thigh_twist_01_r}}
     {targetBone {boneName=deform_groin_r}}
    ]
    constraintWeights [1,1,1,1,1,1,1,1]
   }
   n7 = Anim Dynamics - Bone: dyn_pouch(ComponentPose=@n6, Alpha=0.8, ExternalForce="0,0,0") {
    angularDampingOverride=0.9999
    angularSpringConstant=150
    bAngularSpring=true
    bGravityOverrideInSimSpace=true
    bOverrideAngularDamping=true
    simulationSpace=RootRelative
    alphaBoolBlend {blendOption=Linear}
    boundBone {boneName=dyn_pouch}
    componentAppliedLinearAccClamp=<1500 1500 1500>
    componentLinearAccScale=<0.8 0.8 0.8>
    componentLinearVelScale=<0.8 0.8 0.8>
    physicsBodyDefinitions [{
      sphereCollisionRadius=0
      boundBone {boneName=dyn_pouch}
      boxExtents=<10 20 20>
      constraintSetup {angularLimitsMax=<1 10 0>, angularLimitsMin=<-1 -10 -15>, angularTarget=<1 0 0>}
      localJointOffset=<2 0 0>
     }]
   }
   n8 = Control Rig!(Source=Component To Local(ComponentPose=@n7), Alpha, skydive or emote:double=Get skydive/emote() -> (skydive/emote:double)) -> (Pose) {
    alphaBoolBlend {blendOption=Linear}
    controlRigAssetReference {blueprintRigClass=&/M_LRG_BrawnyBass_Style_CtrlRig.M_LRG_BrawnyBass_Style_CtrlRig_C}
   }
   n9 = Control Rig!(Source=@n8, Alpha) -> (Pose) {
    alphaBoolBlend {blendOption=Linear}
    controlRigAssetReference {blueprintRigClass=/BRCosmetics/Characters/Player/Male/Large/Bodies/M_LRG_BrawnyBass/Meshes/M_LRG_BrawnyBass_Arms_CtrlRig.M_LRG_BrawnyBass_Arms_CtrlRig_C}
   }
   Output Pose(Result=@n9) {layerGroup=None}
  ]}
 EventGraph {nodes [
   Event BlueprintUpdateAnimation()
   > n1 = Branch(Condition=OR Boolean(A=Get bIsSkyDiving(), B=Get bIsPlayingEmote()))
   > Set skydive/emote(skydive/emote:double=1) -> (Output_Get:double)
   Set skydive/emote(execute=@n1.else, skydive/emote=0)
  ]}
]
variables [skydive/emote {type=number}]
